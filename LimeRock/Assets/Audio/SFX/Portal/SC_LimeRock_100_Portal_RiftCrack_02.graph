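# Sound cue for the second LimeRock portal rift crack effect, playing a "cracks in the sky" reality rift wave.
# A single Wave Player is the root, with no randomization or modulation.
# It uses the large POI attenuation asset for long-range audibility, the Fort_Events sound class and reduced volume.
# Per-owner concurrency override values are stored, but the override switch is not set and the concurrency set holds only a null entry.
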SoundCue /LimeRock/Assets/Audio/SFX/Portal/SC_LimeRock_100_Portal_RiftCrack_02
root=WavePlayer_0
nodes [{id=WavePlayer_0, properties {soundWaveAssetPtr=/LimeRock/Assets/Audio/SFX/Portal/Wav/LimeRock_100_RealityRifts_CracksInTheSky_02}}]
properties {
 attenuationSettings=/LimeRock/Assets/Audio/Settings/ATT_LimeRock_POI_Large
 maxDistance=40000
 soundClassObject=&Fort_Events
 volumeMultiplier=0.4
 concurrencyOverrides {bLimitToOwner=true, maxCount=2, retriggerTime=0.25}
 concurrencySet [null]
}
