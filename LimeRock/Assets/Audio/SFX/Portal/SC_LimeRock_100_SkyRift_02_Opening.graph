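# Sound cue for the LimeRock Sky Rift portal opening effect.
# A single wave player feeds the opening wave straight to the output at reduced volume, with no randomization or mixing.
# It uses the LimeRock rift opening attenuation over a long max distance and the Fort events sound class.
# Concurrency overrides set a small per-owner voice cap and a short retrigger time.
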SoundCue /LimeRock/Assets/Audio/SFX/Portal/SC_LimeRock_100_SkyRift_02_Opening
root=WavePlayer_1
nodes [{id=WavePlayer_1, properties {soundWaveAssetPtr=/LimeRock/Assets/Audio/SFX/Portal/Wav/LimeRock_100_SkyRift_02_Opening}}]
properties {
 attenuationSettings=/LimeRock/Assets/Audio/Settings/ATT_LimeRock_RiftOpening
 maxDistance=23000
 soundClassObject=&Fort_Events
 volumeMultiplier=0.6
 concurrencyOverrides {bLimitToOwner=true, maxCount=2, retriggerTime=0.25}
 concurrencySet [null]
}
